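# Animation Blueprint for the SolidMist pickaxe that plays a floating effect on the weapon whenever no montage is playing.
# Each update sets IsFloating from that montage check, driving a two-state machine that switches between an additive identity and a floating loop.
# That output is additively applied to a static idle pose, gated by the same flag, then passes through DefaultSlot.
# A final Modify Bone replaces the two-handed handle bone's translation and rotation.
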AnimBlueprint /BRCosmetics/Weapons/FORT_Melee/Pickaxe_SolidMist/Animations/SolidMist_Pickaxe_BP
target_skeleton=/BRCosmetics/Weapons/FORT_Melee/Pickaxe_SolidMist/Meshes/SolidMist_Axe_Skeleton
graphs [
 AnimGraph {nodes [
   AnimGraphNode_StateMachine_1 = Floating effect() {is=title AnimGraphNode_StateMachine}
   n1 = Apply Additive!(
    Base=Sequence Player 'SolidMist_Idle_Pose_WEP'() {bLoopAnimation=false, sequence=&/SolidMist_Idle_Pose_WEP}
    Additive=@AnimGraphNode_StateMachine_1
    bAlphaBoolEnabled:bool=Get IsFloating() -> (IsFloating:bool)
   ) -> (Pose) {alphaInputType=Bool, alphaBoolBlend {blendOption=Linear}}
   n2 = "Transform (Modify) Bone - Bone: Melee_TwoHanded_Handle"(ComponentPose=Local To Component(LocalPose=Slot 'DefaultSlot'(Source=@n1)), Alpha=1, Translation="0,0,0", Rotation="2,90,-7", Scale="1,1,1") {
    rotationMode=BMM_Replace
    translationMode=BMM_Replace
    alphaBoolBlend {blendOption=Linear}
    boneToModify {boneName=Melee_TwoHanded_Handle}
    rotation=<rot 2 90 -7>
   }
   Output Pose(Result=Component To Local(ComponentPose=@n2)) {layerGroup=None}
   Save cached pose 'Locomotion'(Pose=Input Pose()) {cacheName=Locomotion}
  ]}
 Floating effect {
  entry=Identity
  path=AnimGraph.AnimGraphNode_StateMachine_1.Floating effect
  ids=id
  nodes [
   AnimStateNode_0 = Identity() {is=title AnimStateNode}
   AnimStateNode_2 = Is Floating() {is=title AnimStateNode}
   AnimStateTransitionNode_0 = AnimStateTransitionNode()
   AnimStateTransitionNode_3 = AnimStateTransitionNode() {crossfadeDuration=0.1}
  ]
  links [
   {
    dst=Is Floating
    rule=AnimGraph.AnimGraphNode_StateMachine_1.Floating effect.AnimStateTransitionNode_3.Transition
    src=Identity
    via=AnimStateTransitionNode_3
   }
   {
    dst=Identity
    rule=AnimGraph.AnimGraphNode_StateMachine_1.Floating effect.AnimStateTransitionNode_0.Transition
    src=Is Floating
    via=AnimStateTransitionNode_0
   }
  ]
 }
 Identity {
  path=AnimGraph.AnimGraphNode_StateMachine_1.Floating effect.AnimStateNode_0.Identity
  nodes [Output Animation Pose(Result=Additive Identity Pose())]
 }
 Is Floating {
  path=AnimGraph.AnimGraphNode_StateMachine_1.Floating effect.AnimStateNode_2.Is Floating
  nodes [Output Animation Pose(Result="Sequence Player (additive) 'SolidMist_Floating_Loop_Anim_WEP'"() {sequence=&/SolidMist_Floating_Loop_Anim_WEP})]
 }
 Transition {
  path=AnimGraph.AnimGraphNode_StateMachine_1.Floating effect.AnimStateTransitionNode_0.Transition
  nodes [Result(bCanEnterTransition=NOT Boolean(A=Get IsFloating()))]
 }
 Transition {
  path=AnimGraph.AnimGraphNode_StateMachine_1.Floating effect.AnimStateTransitionNode_3.Transition
  nodes [Result(bCanEnterTransition=Get IsFloating())]
 }
 EventGraph {nodes [
   Event BlueprintUpdateAnimation()
   > n1 = Branch(Condition=IsAnyMontagePlaying!(self:Anim Instance Object Reference) -> (ReturnValue:bool))
   > Set IsFloating(IsFloating:bool=false) -> (Output_Get:bool)
   Set IsFloating(execute=@n1.else, IsFloating=true)
   TryGetPawnOwner()
  ]}
]
variables [IsFloating {type=boolean}]
