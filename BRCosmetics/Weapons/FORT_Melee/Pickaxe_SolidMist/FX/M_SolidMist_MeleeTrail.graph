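# Translucent, unlit, two-sided Niagara ribbon material for the SolidMist pickaxe melee trail, drawing a golden fire streak.
# One layer samples a fire curve atlas along a noise-warped gradient, with camera-distance-dependent edge masking.
# The other pans a fire trail texture, distorted by normal noise scrolling oppositely on each V half, through a second curve.
# Dynamic parameters drive growth, fades, erosion and U offset; particle color tints emissive, and depth fade softens opacity.
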Material /BRCosmetics/Weapons/FORT_Melee/Pickaxe_SolidMist/FX/M_SolidMist_MeleeTrail
graph {
 properties {
  bAutomaticallySetUsageInEditor=false
  bUsedWithNiagaraRibbons=true
  blendMode=BLEND_Translucent
  shadingModel=MSM_Unlit
  twoSided=true
  usedShadingModels=Unlit
 }
 nodes [
  Constant() {r=1}
  n1 = MaterialFunctionCall(XY=TextureCoordinate()) {materialFunction=&Swizzle}
  n2 = TextureCoordinate()
  n3 = Constant() {r=-0.45}
  n4 = Constant() {r=1.5}
  n5 = ScalarParameter() {parameterName=PanSpeed}
  n6 = Constant() {r=-1.6}
  n7 = Panner(
   Coordinate=Multiply(
    A=MaterialFunctionCall(XY=@n2) {materialFunction=&Swizzle}.YX
    B=AppendVector(A=@n3, B=@n4)
   )
   Speed=Multiply(A=Constant2Vector() {g=-0.3, r=0.1}, B=Multiply(A=@n5, B=@n6))
  ) {speedX=-0.02, speedY=-0.3}
  n8 = Constant() {r=2}
  n9 = Constant() {r=7}
  n10 = MaterialFunctionCall() {materialFunction=&LinearGradient}
  n11 = Saturate(None=Sine(None=Divide(A=@n10.VGradient)))
  n12 = ComponentMask(None=Divide(
    A=TextureSample(UVs=@n7) {samplerType=SAMPLERTYPE_Normal, texture=/BRCosmetics/Effects/Fort_Effects/Emotes/UltraLight/Texture/T_DirectionalNoise_Normal}.RGB
    B=LinearInterpolate(A=@n8, B=@n9, Alpha=@n11) {constA=1, constB=4}
   )) {g=true, r=true}
  n13 = CurveAtlasRowParameter(CurveTime=MaterialFunctionCall(UV Channel=Add(A=@n1.YX, B=@n12)) {materialFunction=&LinearGradient}.UGradient) {
   atlas=/Game/Effects/Fort_Effects/GameplayPulginsAssets/Demon/CA_GoldenFire
   curve=/Game/Effects/Fort_Effects/GameplayPulginsAssets/Demon/Curve_GoldenFire01
   parameterName=Param_3
  }
  n14 = Divide(A=MaterialFunctionCall() {materialFunction=&LinearGradient}.VGradient)
  n15 = Constant() {r=2000}
  n16 = Multiply(
   A=Saturate(None=Sine(None=@n14))
   B=LinearInterpolate(
    A=Constant() {r=5}
    B=Constant() {r=100}
    Alpha=MaterialFunctionCall(Distance=@n15) {materialFunction=&Near_Camera_Fade}.Result
   ) {constA=2, constB=100}
  ) {constB=2}
  n17 = Subtract(A=Multiply(A=@n13.A), B=OneMinus(None=Saturate(None=@n16))) {constB=0.1}
  n18 = MaterialFunctionCall() {materialFunction=&LinearGradient}
  n19 = LinearInterpolate(Alpha=DynamicParameter() {parameterIndex=1, paramNames [Grow, Emissive, Width, SpiralWPODirection]}.Grow) {constA=10, constB=0}
  n20 = Subtract(A=Subtract(A=Saturate(None=@n17), B=Multiply(A=@n18.UGradient, B=@n19)) {constB=0.1}, B=Constant() {r=0.1}) {constB=0.1}
  n21 = Multiply(A=Power(Base=Saturate(None=@n20), Exp=Constant() {r=3}), B=Multiply(A=Constant() {r=2}, B=ParticleColor().A))
  n22 = MaterialFunctionCall(UV Channel=TextureCoordinate() {coordinateIndex=1}) {materialFunction=&LinearGradient}
  n23 = OneMinus(None=@n22.VGradient)
  n24 = Constant() {r=1}
  n25 = DynamicParameter() {paramNames [BottomFade,HeadFade,Erode,UOffset]}
  n26 = Multiply(
   A=Power(Base=OneMinus(None=@n22.UGradient), Exp=Constant() {r=1.5})
   B=LinearInterpolate(B=Power(Base=@n23, Exp=@n24), Alpha=@n25.BottomFade) {constA=1}
  )
  n27 = Multiply(A=MaterialFunctionCall() {materialFunction=&LinearGradient}.UGradient)
  n28 = LinearInterpolate(A=Multiply(A=@n26, B=Saturate(None=@n27)), B=@n26, Alpha=@n25.HeadFade)
  n29 = Constant() {r=3}
  n30 = Constant() {r=1.6}
  n31 = Multiply(A=TextureCoordinate(), B=Divide(A=AppendVector(A=@n29, B=@n30), B=Constant() {r=1}))
  n32 = Constant() {r=0.5}
  n33 = Constant() {r=3}
  n34 = Constant() {r=3}
  n35 = Divide(A=AppendVector(A=@n32, B=@n33), B=@n34)
  n36 = ComponentMask(None=TextureSample(UVs=Panner(Coordinate=@n31, Speed=@n35) {speedX=0.2, speedY=0.2}) {samplerType=SAMPLERTYPE_Normal, texture=/Game/Global/Textures/Default/Noise/T_Normal_Noise01}.RGB) {g=true, r=true}
  n37 = Round(None=MaterialFunctionCall() {materialFunction=&LinearGradient}.VGradient)
  n38 = OneMinus(None=@n37)
  n39 = Divide(A=AppendVector(A=@n32, B=Multiply(A=@n33) {constB=-1}), B=@n34)
  n40 = ComponentMask(None=TextureSample(UVs=Panner(Coordinate=@n31, Speed=@n39) {speedX=0.2, speedY=-0.2}) {samplerType=SAMPLERTYPE_Normal, texture=&T_Normal_Noise01}.RGB) {g=true, r=true}
  n41 = Multiply(
   A=Add(A=Multiply(A=@n36, B=@n38), B=Multiply(A=@n40, B=@n37))
   B=LinearInterpolate(Alpha=ScalarParameter() {defaultValue=0.3, parameterName=UV_Distortion}) {constB=0.1}
  )
  n42 = TextureCoordinate()
  n43 = Constant() {r=1}
  n44 = Constant() {r=1.2}
  n45 = MaterialFunctionCall(Float2=MaterialFunctionCall(UVs=Add(A=@n41, B=@n42), Texture Scale=AppendVector(A=@n43, B=@n44)) {materialFunction=&ScaleUVsByCenter}.UVs) {materialFunction=&BreakOutFloat2Components}
  n46 = Add(A=@n45.R, B=DynamicParameter() {paramNames [BottomFade,HeadFade,Erode,UOffset]}.UOffset)
  n47 = ScalarParameter() {parameterName=PanSpeed}
  n48 = Constant()
  n49 = TextureSample(UVs=Panner(Coordinate=AppendVector(A=@n46, B=@n45.G), Speed=AppendVector(A=@n47, B=@n48))) {texture=/Game/Effects/Fort_Effects/Textures/Fire/T_Stylized_Fire_Trail}
  n50 = ParticleColor()
  n51 = Subtract(
   A=Multiply(A=@n49.R, B=Multiply(A=@n50.A, B=@n28))
   B=LinearInterpolate(Alpha=DynamicParameter() {paramNames [BottomFade,HeadFade,Erode,UOffset]}.Erode) {constA=1, constB=0}
  )
  n52 = ScalarParameter() {defaultValue=2, parameterName=Opacity}
  DepthFade(Opacity=Add(A=Multiply(A=@n21, B=@n28), B=Multiply(A=@n51, B=@n52))) {fadeDistanceDefault=20}
  n53 = OneMinus(None=MaterialFunctionCall(UV Channel=TextureCoordinate() {coordinateIndex=1}) {materialFunction=&LinearGradient}.UGradient)
  n54 = Constant() {r=5}
  n55 = Multiply(A=@n13, B=Multiply(A=Constant() {r=3}, B=Power(Base=@n53, Exp=@n54)))
  n56 = CurveAtlasRowParameter(CurveTime=@n49.R) {
   atlas=&CA_GoldenFire
   curve=/Game/Effects/Fort_Effects/GameplayPulginsAssets/Demon/Curve_GoldenFire03
   defaultValue=2
   parameterName=Param_2
  }
  n57 = Multiply(
   A=Multiply(A=Multiply(A=@n56), B=VectorParameter() {parameterName=Color, defaultValue=<rgba 1 0.9052 0.3503 1>}.RGB)
   B=ScalarParameter() {defaultValue=5, parameterName=Emissive}
  )
  MaterialFunctionCall(Emissive=Multiply(A=Add(A=@n55, B=@n57), B=@n50.RGB) {constA=1}) {materialFunction=&MF_FX_EmissiveAdjustments}
  MaterialFunctionCall() {materialFunction=&DebugTimeSine}
  NamedRerouteDeclaration(None=@n41) {name=UVDistortionOpposite, nodeColor=<rgba 0 0.7656 1 1>}
 ]
}
